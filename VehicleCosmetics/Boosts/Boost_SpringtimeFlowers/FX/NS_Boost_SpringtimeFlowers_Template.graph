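# Niagara template for the Springtime Flowers vehicle boost cosmetic, trailing vine ribbons, flower sprites and spinning leaves from the car's boost ports.
# Each emitter uses the filtered socket count to distribute particles across skeletal mesh sockets, then pushes them backward while inherited car movement fades.
# Paint variants tint all three layers, with two-tone random picks for most flower colors.
# A frontend preview flag retunes lifetimes, speeds and material parameters.
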NiagaraSystem /VehicleCosmetics/Boosts/Boost_SpringtimeFlowers/FX/NS_Boost_SpringtimeFlowers_Template
dependencies {
 usedDataInterfaces [
  &NiagaraDataInterfaceCurve
  &NiagaraDataInterfaceArrayFloat
  &NiagaraDataInterfaceArrayColor
  &NiagaraDataInterfaceSpriteRendererInfo
  &NiagaraDataInterfaceSkeletalMesh
 ]
 usedDynamicInputs []
 usedModules [
  &self:SystemSpawnScript.SystemScriptSource.SystemScriptGraph.NiagaraNodeAssignment_0.SetVariables_#1
  &SystemState
 ]
}
emitters {
 NE_VC_Boost_Ribbon_Vine {
  emitterUpdateScript [
   &EmitterState(Life Cycle Mode=::NewEnumerator1, Loop Duration=1, Recalculate Duration Each Loop=true, UseLoopDelay=false)
   &VC_FilteredSocketCount()
   &VC_SpawnRibbon_v1(Number of Ribbons=$Emitter.PortCount, SpawnRate=60)
  ]
  particleSpawnScript [
   &VC_PaintedVariantsOverride(
    Black=<rgba 0.1947 0.4 0.12 1>
    BurntSienna=<rgba 0.24 0.4 0.08 1>
    Cobalt=<rgba 0.225 0.45 0.2625 1>
    Crimson=<rgba 0.2133 0.4 0.08 1>
    DefaultColor=&MakeCustomLinearColorFromBool(Bool=$User.bIsFrontendPreview, True Color=<rgba 0.3878 0.65 0.2925 1>, False Color=<rgba 0.2067 0.5 0.1 1>)
    ForestGreen=<rgba 0.09 0.45 0.09 1>
    Grey=<rgba 0.2355 0.45 0.1575 1>
    Lime=<rgba 0.2612 0.55 0.165 1>
    Orange=<rgba 0.3333 0.5 0.1 1>
    Pink=<rgba 0.2533 0.4 0.2 1>
    Purple=<rgba 0.252 0.45 0.18 1>
    Saffron=<rgba 0.24 0.45 0.09 1>
    SkyBlue=<rgba 0.2 0.4 0.2333 1>
    TitaniumWhite=<rgba 0.3025 0.55 0.22 1>
   )
   &VC_InitializeRibbon_v1(
    Color=$Particles.Color
    Custom Offset=true
    Enable Skeletal Position=true
    Lifetime=&MakeCustomFloatFromBool(Bool=$User.bIsFrontendPreview, True Float=1.25, False Float=0.5)
    Local Vector Offset=&AddVectorToPosition(Position=<25 0 0>, Vector=$User.DTCarOffset)
    Offset Amount=1
    Ribbon Width=80
   )
   &AddVelocity(
    Velocity=&MakeCustomVectorFromBool(Bool=$User.bIsFrontendPreview, True Vector=<1100 0 0>, False Vector=<2000 0 0>)
    Velocity Speed Scale=-1
   ) {module=AddVelocity001}
  ]
  particleUpdateScript [
   &ParticleState()
   &InheritSourceMovement(
    Applied Position Scale=&Multiply_VectorByFloat(
     Vector=<1 1 1>
     Float=&FloatFromCurve(FloatCurve=DataInterface() {propertyValues {Curve {keys [<at 0.2 1 RCIM_Cubic>, {arriveTangent=-1.5383, interpMode=RCIM_Cubic, leaveTangent=-1.5383, tangentMode=RCTM_User, time=1, value=0}]}}})
    )
    ApplyForce=false
    ApplyPosition=true
    ApplyPositionFalloff=false
   )
   &SolveForcesAndVelocity(Clamp Velocity=false, Limit Acceleration=false)
   &VC_UpdateRibbon_v1(Enable Skeletal Position=true)
   &DynamicMaterialParameters(
    Index 0 Param 1=&MakeCustomFloatFromBool(Bool=$User.bIsFrontendPreview, True Float=10, False Float=7)
    Index 0 Param 2=1
    Index 0 Param 3=&Multiply_Float(
     A=&SelectFloatFromArray(Direct Array Index=$Particles.SpawnGroup, Float Selection Array=DataInterface() {propertyValues {FloatData [4,-4]}})
     B=&MakeCustomFloatFromBool(Bool=$User.bIsFrontendPreview, True Float=1, False Float=2)
    )
    Index 0 Param 4=1
    Module.Index0Param0Write=true
    Module.Index0Param1Write=true
    Module.Index0Param2Write=true
    Module.Index0Param3Write=true
   ) {module=DynamicMaterialParameters001}
   &ScaleColor(
    Scale Alpha=&FloatFromCurve(FloatCurve=DataInterface() {propertyValues {Curve {keys [
         {arriveTangent=4.1785, interpMode=RCIM_Cubic, leaveTangent=4.1785, tangentMode=RCTM_User, time=0, value=0}
         {arriveTangent=0.0792, interpMode=RCIM_Cubic, leaveTangent=0.0792, tangentMode=RCTM_User, time=0.15, value=1}
         {arriveTangent=0.0224, interpMode=RCIM_Cubic, leaveTangent=0.0224, tangentMode=RCTM_User, time=1, value=1}
        ]}}})
    ScaleA=true
    ScaleRGB=true
   )
   &ScaleRibbonWidth(Ribbon Width Scale=&SelectFloatFromArray(Direct Array Index=$Particles.SpawnGroup, Float Selection Array=DataInterface() {propertyValues {FloatData [-1,1]}}))
  ]
  rendererClasses [&NiagaraRibbonRendererProperties]
  renderers [{rendererClass=&NiagaraRibbonRendererProperties}]
 }
 NE_VC_Boost_Sprite_Flowers {
  emitterUpdateScript [
   &EmitterState(Life Cycle Mode=::NewEnumerator1, Loop Duration=1, UseLoopDelay=false)
   &VC_FilteredSocketCount()
   &SpawnRate(
    Spawn Probability=2
    SpawnRate=&MakeCustomFloatFromBool(Bool=$User.bIsFrontendPreview, True Float=120, False Float=240)
    Use Spawn Probability=true
   )
  ]
  particleSpawnScript [
   &VC_PaintedVariantsOverride(
    Black=<rgba 0 0 0 100>
    BurntSienna=&RandomRangeLinearColor(Minimum=<rgba 0.08 0 0.0027 1>, Maximum=<rgba 0.0219 0.0125 0.05 1>)
    Cobalt=&SelectLinearColorFromArray(
     Array Sampling Mode=::NewEnumerator0
     Direct Array Index=?
     Color Selection Array=DataInterface() {propertyValues {ColorData [<rgba 0.0625 0.1469 0.25 1>, <rgba 0.0756 0.1897 0.2 1>]}}
     Random Seed=0
    )
    Crimson=&SelectLinearColorFromArray(
     Array Sampling Mode=::NewEnumerator0
     Direct Array Index=?
     Color Selection Array=DataInterface() {propertyValues {ColorData [<rgba 0.15 0.0225 0.065 1>, <rgba 0.0358 0.0075 0.05 1>]}}
     Random Seed=0
    )
    DefaultColor=<rgba 5 5 5 1>
    ForestGreen=&SelectLinearColorFromArray(
     Array Sampling Mode=::NewEnumerator0
     Direct Array Index=?
     Color Selection Array=DataInterface() {propertyValues {ColorData [<rgba 0.0384 0.09 0.018 1>, <rgba 0.08 0.0517 0.012 1>]}}
     Random Seed=0
    )
    Grey=&SelectLinearColorFromArray(
     Array Sampling Mode=::NewEnumerator0
     Direct Array Index=?
     Color Selection Array=DataInterface() {propertyValues {ColorData [<rgba 0.08 0.0728 0.068 1>, <rgba 0.08 0.08 0.08 1>]}}
     Random Seed=0
    )
    Lime=&SelectLinearColorFromArray(
     Array Sampling Mode=::NewEnumerator0
     Direct Array Index=?
     Color Selection Array=DataInterface() {propertyValues {ColorData [<rgba 0.38 0.4 0.16 1>, <rgba 0.2 0.126 0.12 1>]}}
     Random Seed=0
    )
    Orange=&SelectLinearColorFromArray(
     Array Sampling Mode=::NewEnumerator0
     Direct Array Index=?
     Color Selection Array=DataInterface() {propertyValues {ColorData [<rgba 0.45 0.1935 0.1978 1>, <rgba 0.1338 0.075 0.15 1>]}}
     Random Seed=0
    )
    Pink=&SelectLinearColorFromArray(
     Array Sampling Mode=::NewEnumerator0
     Direct Array Index=?
     Color Selection Array=DataInterface() {propertyValues {ColorData [<rgba 0.35 0.1925 0.2923 1>, <rgba 0.1335 0.13 0.2 1>]}}
     Random Seed=0
    )
    Purple=&SelectLinearColorFromArray(
     Array Sampling Mode=::NewEnumerator0
     Direct Array Index=?
     Color Selection Array=DataInterface() {propertyValues {ColorData [<rgba 0.0608 0.045 0.15 1>, <rgba 0.065 0.1 0.1 1>]}}
     Random Seed=0
    )
    Saffron=&SelectLinearColorFromArray(
     Array Sampling Mode=::NewEnumerator0
     Direct Array Index=?
     Color Selection Array=DataInterface() {propertyValues {ColorData [<rgba 0.3 0.24 0.12 1>, <rgba 0.2 0.12 0.132 1>]}}
     Random Seed=0
    )
    SkyBlue=&SelectLinearColorFromArray(
     Array Sampling Mode=::NewEnumerator0
     Direct Array Index=?
     Color Selection Array=DataInterface() {propertyValues {ColorData [<rgba 0.16 0.34 0.4 1>, <rgba 0.055 0.1 0.0625 1>]}}
     Random Seed=0
    )
    TitaniumWhite=&SelectLinearColorFromArray(
     Array Sampling Mode=::NewEnumerator0
     Direct Array Index=?
     Color Selection Array=DataInterface() {propertyValues {ColorData [<rgba 0.7 0.7 0.7 1>, <rgba 0.22 0.22 0.22 1>]}}
     Random Seed=0
    )
   )
   &InitializeParticle(
    Color=$Particles.Color
    Lifetime Max=&MakeCustomFloatFromBool(Bool=$User.bIsFrontendPreview, True Float=2, False Float=0.75)
    Lifetime Min:NiagaraFloat=&MakeCustomFloatFromBool(Bool=$User.bIsFrontendPreview, True Float=1, False Float=0.5)
    Lifetime Mode=::NewEnumerator1
    Sprite Rotation Mode=::NewEnumerator2
    Sprite Size Mode=::NewEnumerator1
    Uniform Sprite Size Max=60
    Uniform Sprite Size Min=30
    UsePositionOffset=false
    Write Material Random=true
   )
   &PartitionParticles(Number Of Partitions=$Emitter.PortCount)
   &SkeletalMeshLocation(
    Bone / Socket Index=$Particles.PartitionParticles.Partition
    Bone / Socket Sampling Mode=::NewEnumerator5
    MaskSpawnGroup=false
    Mesh Sampling Type=::NewEnumerator6
    Sampled Position Offset=&AddVectorToPosition(Position=<-30 0 0>, Vector=$User.DTCarOffset)
    Skeletal Mesh=$System.SkeletalMesh
   )
   &ShapeLocation(MaskSpawnGroup=false, Override Seed=false, Sphere Radius=40, UseOwnerScale=true)
   &AddVelocity(
    Cone Angle=15
    Cone Axis=&RandomRangeVector(Minimum=<-1 0 0.05>, Maximum=<-1 0 0>)
    Module.UseConeFalloff=false
    Module.UseVelDistribution=false
    Override Seed=false
    Velocity Mode=::NewEnumerator2
    Velocity Speed=1000
   )
  ]
  particleUpdateScript [
   &CameraOffset(Camera Offset Amount=&RandomRangeFloat(Minimum=-2, Maximum=-3))
   &InheritSourceMovement(
    Applied Position Scale=&Multiply_VectorByFloat(
     Vector=<1 1 1>
     Float=&FloatFromCurve(FloatCurve=DataInterface() {propertyValues {Curve {keys [
          {interpMode=RCIM_Cubic, tangentMode=RCTM_User, time=0, value=1}
          {arriveTangent=-1.3944, interpMode=RCIM_Cubic, leaveTangent=-1.3944, tangentMode=RCTM_User, time=1, value=0.6}
         ]}}})
    )
    ApplyForce=false
    ApplyPosition=true
    ApplyPositionFalloff=false
   )
   &DynamicMaterialParameters(
    Index 0 Param 1=&FloatFromCurve(FloatCurve=DataInterface() {propertyValues {Curve {keys [
         {arriveTangent=-0.0939, interpMode=RCIM_Cubic, leaveTangent=-0.0939, leaveTangentWeight=0.669, tangentMode=RCTM_User, time=0, value=2}
         {arriveTangent=-0.0149, interpMode=RCIM_Cubic, leaveTangent=-0.0149, tangentMode=RCTM_User, time=1.5, value=1}
        ]}}})
    Index 0 Param 2=&MakeCustomFloatFromBool(
     Bool=$User.bIsFrontendPreview
     True Float=&MakeCustomFloatFromBool(Bool=$User.bIsPainted, True Float=&RandomRangeFloat(Minimum=2.5, Maximum=3), False Float=0.5)
     False Float=&MakeCustomFloatFromBool(Bool=$User.bIsPainted, True Float=2, False Float=1)
    )
    Index 0 Param 3=&MakeCustomFloatFromBool(
     Bool=$User.bIsFrontendPreview
     True Float=&MakeCustomFloatFromBool(Bool=$User.bIsPainted, True Float=&RandomRangeFloat(Minimum=1.2, Maximum=1.75), False Float=0)
     False Float=&MakeCustomFloatFromBool(Bool=$User.bIsPainted, True Float=2, False Float=0)
    )
    Index 0 Param 4=&MakeCustomFloatFromBool(
     Bool=$User.bIsFrontendPreview
     True Float=&MakeCustomFloatFromBool(Bool=$User.bIsPainted, True Float=&RandomRangeFloat(Minimum=1, Maximum=1.2), False Float=6.5)
     False Float=1
    )
    Module.Index0Param0Write=true
    Module.Index0Param1Write=true
    Module.Index0Param2Write=true
    Module.Index0Param3Write=true
   )
   &SubUVAnimation(
    Direct Frame Index:NiagaraFloat=&RandomRangeFloat(Minimum=4, Maximum=0)
    "Sprite Renderer "=DataInterface() {propertyValues {SpriteRenderer=&self:NE_VC_Boost_Sprite_Flowers_2.NiagaraSpriteRendererProperties_0}}
    SubUV Animation Mode:ENiagara_SubUVLookupModeV2=::NewEnumerator5
   )
   &ParticleState()
   &ScaleSpriteSize(Uniform Curve Sprite Scale=DataInterface() {propertyValues {Curve {keys [
        {interpMode=RCIM_Cubic, tangentMode=RCTM_Break, time=0, value=1}
        {interpMode=RCIM_Cubic, leaveTangent=-1.7361, tangentMode=RCTM_Break, time=0.15, value=1.5}
        {arriveTangent=-1.7305, tangentMode=RCTM_User, time=1, value=0}
       ]}}}) {enabled=false, module=ScaleSpriteSize001}
   &SpriteRotationRate(
    Delta Time:NiagaraFloat=$Particles.Initial.Lifetime
    Rotation Rate=&RandomRangeFloat(Minimum=2.5, Maximum=-2.5)
   )
   &ScaleSpriteSize(
    Scale Sprite Size Mode=::NewEnumerator0
    Uniform Scale Factor=&MakeCustomFloatFromBool(
     Bool=$User.bIsFrontendPreview
     True Float=&FloatFromCurve(FloatCurve=DataInterface() {propertyValues {Curve {keys [
          {interpMode=RCIM_Cubic, tangentMode=RCTM_Break, time=0, value=1}
          {interpMode=RCIM_Cubic, leaveTangent=-1.7361, tangentMode=RCTM_Break, time=0.15, value=1.5}
          {arriveTangent=-1.7305, tangentMode=RCTM_User, time=1, value=0}
         ]}}})
     False Float=&FloatFromCurve(FloatCurve=DataInterface() {propertyValues {Curve {keys [
          {interpMode=RCIM_Cubic, tangentMode=RCTM_Break, time=0, value=1}
          {interpMode=RCIM_Cubic, leaveTangent=-0.2751, tangentMode=RCTM_Break, time=0.15, value=1.5}
          {arriveTangent=-0.4843, tangentMode=RCTM_User, time=1, value=0.5}
         ]}}})
    )
   )
   &WindForce(
    Frequency=0.1
    Offset By Time:Vector3f=<0 0 0>
    Scale=2
    Turbulence Flow Loop Duration=1
    Wind Speed=<100 0 -100>
    Wind Speed Scale=&FloatFromCurve(
     FloatCurve=DataInterface() {propertyValues {Curve {keys [<at 0.5 0 RCIM_Cubic>, {arriveTangent=31.5688, interpMode=RCIM_Cubic, leaveTangent=31.5688, tangentMode=RCTM_User, time=0.8, value=5}]}}}
     Scale Curve=&RandomRangeFloat(Minimum=0, Maximum=-2)
    )
   ) {enabled=false}
   &Drag(Drag=0.5, Use Linear Drag=true, Use Rotational Drag=true)
   &SolveForcesAndVelocity(Clamp Velocity=false, Limit Acceleration=false)
   &ScaleColor(Scale Alpha=0, Scale RGB=&Multiply_VectorByFloat(Vector=<1 1 1>, Float=1), ScaleA=true, ScaleRGB=true)
  ]
  rendererClasses [&NiagaraSpriteRendererProperties]
  renderers [{rendererClass=&NiagaraSpriteRendererProperties}]
 }
 NE_VC_Boost_Sprite_Leafs {
  emitterUpdateScript [
   &EmitterState(Life Cycle Mode=::NewEnumerator1, Loop Duration=1, UseLoopDelay=false)
   &VC_FilteredSocketCount()
   &SpawnRate(SpawnRate=50, Use Spawn Probability=false)
  ]
  particleSpawnScript [
   &VC_PaintedVariantsOverride(
    Black=<rgba 0.1947 0.4 0.12 1>
    BurntSienna=<rgba 0.24 0.4 0.08 1>
    Cobalt=<rgba 0.225 0.45 0.2625 1>
    Crimson=<rgba 0.2133 0.4 0.08 1>
    DefaultColor=&MakeCustomLinearColorFromBool(Bool=$User.bIsFrontendPreview, True Color=<rgba 0.3878 0.65 0.2925 1>, False Color=<rgba 0.2067 0.5 0.1 1>)
    ForestGreen=<rgba 0.09 0.45 0.09 1>
    Grey=<rgba 0.2355 0.45 0.1575 1>
    Lime=<rgba 0.2612 0.55 0.165 1>
    Orange=<rgba 0.3333 0.5 0.1 1>
    Pink=<rgba 0.2533 0.4 0.2 1>
    Purple=<rgba 0.252 0.45 0.18 1>
    Saffron=<rgba 0.24 0.45 0.09 1>
    SkyBlue=<rgba 0.2 0.4 0.2333 1>
    TitaniumWhite=<rgba 0.3025 0.55 0.22 1>
   )
   &InitializeParticle(
    Color=$Particles.Color
    Lifetime Max=&MakeCustomFloatFromBool(Bool=$User.bIsFrontendPreview, True Float=1, False Float=0.5)
    Lifetime Min:NiagaraFloat=&MakeCustomFloatFromBool(Bool=$User.bIsFrontendPreview, True Float=0.5, False Float=0.5)
    Lifetime Mode=::NewEnumerator1
    Sprite Rotation Mode=::NewEnumerator2
    Sprite Size Mode=::NewEnumerator1
    Uniform Sprite Size Max=30
    Uniform Sprite Size Min=20
    UsePositionOffset=false
    Write Material Random=true
   )
   &PartitionParticles(Number Of Partitions=$Emitter.PortCount)
   &SkeletalMeshLocation(
    Bone / Socket Index=$Particles.PartitionParticles.Partition
    Bone / Socket Sampling Mode=::NewEnumerator5
    MaskSpawnGroup=false
    Mesh Sampling Type=::NewEnumerator6
    Sampled Position Offset=&AddVectorToPosition(Position=<0 0 0>, Vector=$User.DTCarOffset)
    Skeletal Mesh=$System.SkeletalMesh
   )
   &AddVelocity(
    Cone Angle=15
    Cone Axis=<-1 0 0.02>
    Module.UseConeFalloff=false
    Module.UseVelDistribution=false
    Override Seed=false
    Velocity Mode=::NewEnumerator2
    Velocity Speed=&Multiply_Float(A=900, B=&MakeCustomFloatFromBool(Bool=$User.bIsFrontendPreview, True Float=1, False Float=2))
   )
   &ShapeLocation(MaskSpawnGroup=false, Override Seed=false, Sphere Radius=20, UseOwnerScale=true)
  ]
  particleUpdateScript [
   &ParticleState()
   &SpriteRotationRate(Delta Time:NiagaraFloat=$Particles.Initial.Lifetime, Rotation Rate=500)
   &ScaleSpriteSize(Uniform Curve Sprite Scale=DataInterface() {propertyValues {Curve {keys [
        {tangentMode=RCTM_User, time=0, value=1}
        {arriveTangent=-0.0405, interpMode=RCIM_Cubic, leaveTangent=-0.0405, tangentMode=RCTM_User, time=0.7009, value=1}
        {arriveTangent=-8.0089, interpMode=RCIM_Cubic, leaveTangent=-8.0089, tangentMode=RCTM_User, time=1, value=0}
       ]}}})
   &InheritSourceMovement(
    Applied Position Scale=&Multiply_VectorByFloat(
     Vector=<1 1 1>
     Float=&FloatFromCurve(FloatCurve=DataInterface() {propertyValues {Curve {keys [
          {interpMode=RCIM_Cubic, tangentMode=RCTM_User, time=0, value=1}
          {arriveTangent=-0.1485, interpMode=RCIM_Cubic, leaveTangent=-0.1485, tangentMode=RCTM_User, time=1, value=0.1}
         ]}}})
    )
    ApplyForce=false
    ApplyPosition=true
    ApplyPositionFalloff=false
   )
   &SolveForcesAndVelocity(Clamp Velocity=false, Limit Acceleration=false)
   &DynamicMaterialParameters(
    Index 0 Param 2=&MakeCustomFloatFromBool(Bool=$User.bIsFrontendPreview, True Float=10, False Float=7)
    Index 0 Param 3=&RandomRangeFloat(Minimum=0, Maximum=30)
    Index 0 Param 4=1
    Module.Index0Param0Write=true
    Module.Index0Param1Write=true
    Module.Index0Param2Write=true
    Module.Index0Param3Write=true
   )
   &SubUVAnimation(
    End Frame Range Override=16
    "Sprite Renderer "=DataInterface() {propertyValues {SpriteRenderer=&self:NE_VC_Boost_Sprite_Leafs_0.NiagaraSpriteRendererProperties_0}}
    SubUV Animation Mode:ENiagara_SubUVLookupModeV2=::NewEnumerator2
    UseEndFrame=true
    UseStartFrame=false
   )
  ]
  rendererClasses [&NiagaraSpriteRendererProperties]
  renderers [{rendererClass=&NiagaraSpriteRendererProperties}]
 }
}
summary {
 systemName=NS_Boost_SpringtimeFlowers_Template
 emitters [
  {emitterName=NE_VC_Boost_Ribbon_Vine, rendererClasses [&NiagaraRibbonRendererProperties]}
  {emitterName=NE_VC_Boost_Sprite_Flowers, rendererClasses [&NiagaraSpriteRendererProperties]}
  {emitterName=NE_VC_Boost_Sprite_Leafs, rendererClasses [&NiagaraSpriteRendererProperties]}
 ]
 userVariables [
  User.bIsFrontendPreview {type=NiagaraBool}
  User.bIsPainted {type=NiagaraBool}
  User.ColorVariantID {type=NiagaraInt32}
  User.DTCarOffset {type=Vector3f, description="", defaultValue {struct=Vector3f, value=<0 0 0>}}
  User.PaintedColor {type=LinearColor, description="", defaultValue {struct=LinearColor, value=<rgba 1 1 1 1>}}
  User.PaintedScalar {type=NiagaraFloat, defaultValue=1, description=""}
  User.PortCount {type=NiagaraInt32, defaultValue=2}
 ]
}
