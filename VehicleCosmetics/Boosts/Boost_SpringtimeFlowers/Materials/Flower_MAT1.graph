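# Unlit, masked, dithered Niagara sprite material for flower particles in the Springtime Flowers vehicle boost.
# Radially mapped, panned rainbow gradients pulse brightness and perturb flower texture rotation, driven by dynamic UVDistort, Brightness, Saturation and Bloom parameters.
# Near-white particle color (red-weighted, green ignored) keeps texture colors; others tint desaturated flowers with optional hue shift, feeding MF_FX_EmissiveAdjustments.
# Flower alpha, radial falloffs and particle alpha form a sine-shaped mask.
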Material /VehicleCosmetics/Boosts/Boost_SpringtimeFlowers/Materials/Flower_MAT1
graph {
 properties {
  bCastDynamicShadowAsMasked=true
  bUsedWithNiagaraSprites=true
  blendMode=BLEND_Masked
  ditherOpacityMask=true
  pixelDepthOffsetMode=PDOM_Legacy
  refractionMethod=RM_2DOffset
  shadingModel=MSM_Unlit
  usedShadingModels=Unlit
 }
 nodes [
  n1 = Add(A=Frac(None=TextureCoordinate() {uTiling=2, vTiling=2})) {constB=-0.5}
  n2 = DotProduct(A=@n1, B=@n1)
  n3 = OneMinus(None=SquareRoot(None=@n2))
  n4 = Constant() {desc=12, r=12}
  n5 = Add(A=Frac(None=TextureCoordinate() {uTiling=2, vTiling=2})) {constB=-0.5}
  n6 = DotProduct(A=@n5, B=@n5)
  n7 = Multiply(A=Clamp(None=ConstantBiasScale(None=@n6) {bias=0.1, scale=1})) {constB=2}
  n8 = DynamicParameter() {paramNames [UVDistort,Brightness,Saturation,Bloom]}
  n9 = TextureSample(UVs=Panner(Coordinate=@n7, Time=@n8.UVDistort) {speedX=-0.5, speedY=-0.5}) {texture=/VehicleCosmetics/Boosts/Boost_SpringtimeFlowers/Textures/T_RainbowGradient}
  n10 = Rotator(Coordinate=TextureCoordinate(), Time=Multiply(A=@n9.B) {constB=0.05}) {speed=1}
  n11 = TextureSample(UVs=@n10) {texture=/VehicleCosmetics/Boosts/Boost_SpringtimeFlowers/Textures/T_Flowers_FB_D}
  n12 = Desaturation(None=@n11.RGB)
  n13 = Multiply(A=@n12, B=Clamp(None=Power(Base=@n3, Exp=@n4)))
  n14 = Clamp(None=ConstantBiasScale(None=OneMinus(None=@n2)) {bias=-0.8, scale=1})
  n15 = Multiply(A=@n14, B=@n14)
  n16 = ParticleColor()
  n17 = Add(A=Add(A=ConstantBiasScale(None=@n13) {bias=0.2, scale=0.6}, B=Clamp(None=@n15)), B=OneMinus(None=Add(A=@n16.A) {constB=0.2}))
  Clamp(None=Multiply(A=Sine(None=@n17) {period=5}, B=@n11.A))
  n18 = ParticleColor()
  n19 = Divide(A=Add(A=Add(A=@n18.R, B=@n18.R), B=@n18.B)) {constB=3}
  n20 = TextureSample(UVs=Panner(Coordinate=@n7, Time=Constant() {r=1}) {speedX=-0.5, speedY=-0.5}) {texture=&T_RainbowGradient}
  n21 = Add(A=@n20.R, B=@n20.R)
  n22 = Multiply(A=Multiply(A=@n21, B=@n8.Brightness), B=@n11.RGB)
  n23 = Multiply(A=Add(A=@n9.R, B=@n9.R), B=@n8.Brightness)
  n24 = Multiply(A=Multiply(A=@n11.RGB, B=@n23), B=@n8.Bloom)
  n25 = Desaturation(None=Multiply(A=Add(A=@n22, B=@n24)) {constB=0.5}, Fraction=Constant() {r=0.05})
  n26 = Multiply(A=@n21, B=@n12)
  n27 = Multiply(A=Multiply(A=Power(Base=@n12), B=@n23), B=@n8.Bloom)
  n28 = Clamp(None=@n8.Bloom)
  n29 = Subtract(A=@n21, B=ScalarParameter() {defaultValue=0.8, parameterName=HueShift_Amount})
  n30 = ScalarParameter() {defaultValue=0.5, parameterName=HueShift_Sharpness}
  n31 = ParticleColor()
  n32 = Add(A=@n31.R, B=@n31.R)
  n33 = Clamp(None=Divide(A=Add(A=@n32, B=@n31.B)) {constB=3})
  n34 = Add(A=@n31.RGB) {constB=0.05}
  n35 = Add(A=@n31.RGB) {constB=0.075}
  n36 = Desaturation(None=@n31.RGB, Fraction=Constant() {r=0.05})
  n37 = LinearInterpolate(
   A=Desaturation(None=If(A=@n33, "A > B"=@n34, "A == B"=@n35, "A < B"=@n35) {constB=0.3}, Fraction=Constant() {r=0.1})
   B=If(A=@n33, "A > B"=@n36, "A == B"=@n36, "A < B"=@n31.RGB) {constB=0.78}
   Alpha=@n31.RGB
  )
  n38 = LinearInterpolate(
   A=MaterialFunctionCall(Hue Shift Percentage=Multiply(A=@n29, B=@n30), Texture=@n37) {materialFunction=&HueShift}.Result
   B=@n37
   Alpha=@n8.Saturation
  )
  n39 = Multiply(A=Multiply(A=Add(A=@n26, B=@n27), B=Multiply(A=@n28, B=@n38)), B=@n8.Brightness)
  MaterialFunctionCall(Emissive=If(A=Clamp(None=@n19), "A > B"=Clamp(None=@n25), "A == B"=@n39, "A < B"=@n39) {constB=0.98}) {materialFunction=&MF_FX_EmissiveAdjustments}
  TextureSample(UVs=@n10) {samplerType=SAMPLERTYPE_Normal, texture=/VehicleCosmetics/Boosts/Boost_SpringtimeFlowers/Textures/T_Flowers_FB_N}
 ]
}
