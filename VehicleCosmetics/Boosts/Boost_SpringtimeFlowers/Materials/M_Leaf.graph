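# Masked unlit Niagara sprite material for the leaves of the Springtime Flowers vehicle boost.
# The red channel of Normal_01 (a _D texture) passes through Desaturation, then is tinted by particle color and scaled by dynamic Brightness into MF_FX_EmissiveAdjustments.
# Particle alpha times Diffuse_01 alpha (an _N texture) is clamped for the mask, so parameter names and textures look swapped.
# AnimationSpeed and FresnelBrightness dynamic parameters are unused.
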Material /VehicleCosmetics/Boosts/Boost_SpringtimeFlowers/Materials/M_Leaf
graph {
 properties {
  bUsedWithNiagaraSprites=true
  blendMode=BLEND_Masked
  pixelDepthOffsetMode=PDOM_Legacy
  refractionMethod=RM_2DOffset
  shadingModel=MSM_Unlit
  usedShadingModels=Unlit
 }
 nodes [
  n1 = ParticleColor()
  n2 = TextureCoordinate() {desc=1x1}
  Clamp(None=Multiply(
    A=@n1.A
    B=TextureSampleParameter2D(UVs=@n2) {parameterName=Diffuse_01, texture=/VehicleCosmetics/Boosts/Boost_SpringtimeFlowers/Textures/T_Leaf_02_FB_N}.A
   ))
  n3 = DynamicParameter() {paramNames [AnimationSpeed, Brightness, FresnelBrightness, Saturation]}
  n4 = Desaturation(
   None=TextureSampleParameter2D(UVs=@n2) {parameterName=Normal_01, texture=/VehicleCosmetics/Boosts/Boost_SpringtimeFlowers/Textures/T_Leaf_02_FB_D}.R
   Fraction=@n3.Saturation
  )
  MaterialFunctionCall(Emissive=Multiply(A=Multiply(A=@n4, B=@n1.RGB), B=@n3.Brightness)) {materialFunction=&MF_FX_EmissiveAdjustments}
 ]
}
