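# Unlit translucent Niagara ribbon material for the thorny vine trail of the Springtime Flowers vehicle boost.
# Parameter-tiled and offset UV1 samples the vine texture, whose red channel is tinted by particle color and dynamic brightness into an emissive adjustment function.
# Opacity combines particle alpha, the normal texture's alpha and a UV0 fade at both ribbon ends.
# Time-driven layered sine and cosine waves wobble vertices, eased in along U, scaled by WPOINTENSITY and gated by the TrailBlend switch.
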Material /VehicleCosmetics/Boosts/Boost_SpringtimeFlowers/Materials/M_VineTrail
graph {
 properties {
  bUsedWithNiagaraRibbons=true
  blendMode=BLEND_Translucent
  pixelDepthOffsetMode=PDOM_Legacy
  refractionMethod=RM_PixelNormalOffset
  shadingModel=MSM_Unlit
  usedShadingModels=Unlit
 }
 nodes [
  n1 = ParticleColor()
  n2 = Multiply(A=ComponentMask(None=TextureCoordinate() {coordinateIndex=1, vTiling=0}) {r=true}, B=ScalarParameter() {defaultValue=1, parameterName=TIleU})
  n3 = ScalarParameter() {parameterName=PanU}
  n4 = Multiply(A=ComponentMask(None=TextureCoordinate() {coordinateIndex=1, uTiling=0}) {g=true}, B=ScalarParameter() {defaultValue=1, parameterName=TIleV})
  n5 = ScalarParameter() {parameterName=PanV}
  n6 = AppendVector(
   A=ComponentMask(None=Panner(Coordinate=@n2, Time=@n3) {speedX=0.5}) {r=true}
   B=ComponentMask(None=Panner(Coordinate=@n4, Time=@n5) {speedY=0.5}) {g=true}
  )
  n7 = TextureSample(UVs=@n6) {texture=/VehicleCosmetics/Boosts/Boost_SpringtimeFlowers/Textures/T_ThornVIne_02_N}
  n8 = ComponentMask(None=TextureCoordinate()) {r=true}
  n9 = Clamp(None=Multiply(B=@n8) {constA=8})
  n10 = Multiply(A=Multiply(B=OneMinus(None=@n8)) {constA=4}, B=@n9)
  Clamp(None=Multiply(A=Multiply(A=@n1.A, B=@n7.A), B=Clamp(None=@n10)))
  n11 = DynamicParameter() {paramNames [Brightness,Saturation,WPOINTENSITY,Parametro4]}
  n12 = Desaturation(None=TextureSample(UVs=@n6) {texture=/VehicleCosmetics/Boosts/Boost_SpringtimeFlowers/Textures/T_ThornVine_02_D}.R, Fraction=@n11.Saturation)
  MaterialFunctionCall(Emissive=Multiply(A=@n11.Brightness, B=Multiply(A=@n1.RGB, B=@n12))) {materialFunction=&MF_FX_EmissiveAdjustments}
  n13 = VectorParameter() {parameterName=WPO_Intensity, defaultValue=<rgba 1 2 1 1>}
  n14 = Multiply(A=@n13.G, B=OneMinus(None=Time()))
  n15 = ComponentMask(None=Multiply(A=TextureCoordinate() {coordinateIndex=1, uTiling=2}, B=@n13.R)) {r=true}
  n16 = Add(A=@n14, B=@n15)
  n17 = Add(A=Cosine(None=@n16), B=Cosine(None=@n16) {period=2})
  n18 = Add(A=@n15, B=@n14)
  n19 = Add(A=Sine(None=@n18), B=Sine(None=@n18) {period=2})
  n20 = Transform(None=AppendVector(A=AppendVector(A=@n17, B=@n19), B=@n17))
  n21 = DynamicParameter() {paramNames [Brightness,Saturation,WPOINTENSITY,Parametro4]}
  StaticSwitchParameter(True=Multiply(A=@n9, B=Multiply(A=@n20, B=@n21.WPOINTENSITY)), False=Constant()) {parameterName=TrailBlend}
 ]
}
